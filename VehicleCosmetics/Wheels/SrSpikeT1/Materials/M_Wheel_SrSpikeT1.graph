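# Masked, two-sided material for the SrSpikeT1 cosmetic vehicle wheel, adding an animated swirl glow and optional tire rubber detail to a shared wheel parent function.
# Panning and rotating swirl samples form the emissive, gated by vertex-red masks selected through the LOD_SE switch, then scaled by brightness and tinted.
# The TireRubberExist switch adds a masked, tiled detail normal.
# Overridden attributes then feed the vehicle damage and de-rez transition functions.
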Material /VehicleCosmetics/Wheels/SrSpikeT1/Materials/M_Wheel_SrSpikeT1
graph {
 properties {
  bAutomaticallySetUsageInEditor=false
  bUseMaterialAttributes=true
  bUsedWithMeshDeformer=true
  bUsedWithSkeletalMesh=true
  blendMode=BLEND_Masked
  pixelDepthOffsetMode=PDOM_Legacy
  twoSided=true
  usedShadingModels=Default Lit
 }
 nodes [
  n1 = MaterialFunctionCall() {materialFunction=/VehicleCosmetics/Wheels/Mage/Materials/MF_WheelParent_Simple_Alpha}
  n2 = GetMaterialAttributes(None=@n1.MaterialAttributes) {attributeGetTypes [#1]}
  n3 = NamedRerouteUsage()
  n4 = NamedRerouteUsage()
  MaterialFunctionCall(In=MaterialFunctionCall(Material=SetMaterialAttributes!(MaterialAttributes=@n2.MaterialAttributes, Emissive Color=@n3, Normal=@n4) {attributeSetTypes [#2, #1]}) {materialFunction=&MF_BaseVehicle_Damage}.Result) {materialFunction=&MF_VehicleDeRez_Transition}
  NamedRerouteDeclaration(None=ComponentMask(None=@n1.RimRGBAMask) {r=true}) {name=TireSEMask, nodeColor=<rgba 1 0 0.9844 1>}
  n5 = TextureSample(UVs=Panner(Coordinate=TextureCoordinate() {uTiling=0.2, unMirrorU=true, vTiling=0.3}) {speedY=0.03}) {texture=/VehicleCosmetics/Wheels/_SharedTextures/FX/Swirl_RGB_02_T}
  n6 = TextureCoordinate() {uTiling=0.5, unMirrorU=true, vTiling=0.5}
  n7 = Multiply(
   A=TextureSample(UVs=Panner(Coordinate=@n6) {speedY=0.06}) {texture=&Swirl_RGB_02_T}.R
   B=TextureSample(UVs=Panner(Coordinate=@n6) {speedY=0.02}) {texture=&Swirl_RGB_02_T}.B
  )
  n8 = ScalarParameter() {defaultValue=1, group=SR_Spike_SE, parameterName=SpiralIntensity, sortPriority=1}
  n9 = Multiply(
   A=TextureSample(UVs=Rotator(Coordinate=@n6)) {texture=&Swirl_RGB_02_T}.B
   B=TextureSample(UVs=Rotator(Coordinate=@n6) {speed=0.2}) {texture=&Swirl_RGB_02_T}.G
  )
  n10 = Constant()
  n11 = NamedRerouteUsage()
  n12 = VertexColor()
  n13 = NamedRerouteUsage()
  n14 = Constant()
  n15 = VertexColor()
  n16 = Multiply(
   A=Add(A=Add(A=@n5.G, B=@n7), B=Multiply(A=@n8, B=@n9))
   B=StaticSwitchParameter(True=LinearInterpolate(A=@n10, B=@n11, Alpha=@n12.R), False=LinearInterpolate(A=@n13, B=@n14, Alpha=@n15.R)) {group=SE_LOD, parameterName=LOD_SE, sortPriority=0}
  )
  n17 = ScalarParameter() {defaultValue=3, group=SR_Spike_SE, parameterName=TireSEBrightness, sortPriority=1}
  n18 = VectorParameter() {group=SR_Spike_SE, parameterName=CustomColor, sortPriority=1, defaultValue=<rgba 0.4793 0.159 0.0222 0.302>}
  NamedRerouteDeclaration(None=Multiply(A=Multiply(A=@n16, B=@n17), B=@n18.RGB)) {name=Emissive, nodeColor=<rgba 1 0 0.7287 1>}
  n19 = NamedRerouteUsage()
  n20 = Multiply(A=TextureCoordinate(), B=ScalarParameter() {defaultValue=2, group=TireRubber, parameterName=TextureDetailNormalTiling})
  n21 = NamedRerouteUsage()
  n22 = Multiply(
   A=Multiply(A=TextureSample(UVs=@n20) {samplerType=SAMPLERTYPE_Normal, texture=/VehicleCosmetics/SharedTextures/T_Matte_N}.RGB, B=OneMinus(None=@n21))
   B=VertexColor().R
  )
  NamedRerouteDeclaration(None=StaticSwitchParameter(True=Add(A=@n19, B=@n22), False=@n19) {group=TireRubber, parameterName=TireRubberExist}) {name=NormalRubber_1, nodeColor=<rgba 0.1016 0 1 1>}
  NamedRerouteDeclaration(None=ComponentMask(None=@n1.TireRGBAMask) {g=true}) {name=TireRubberMask, nodeColor=<rgba 0 0.0391 1 1>}
  NamedRerouteDeclaration(None=ComponentMask(None=@n1.TireRGBAMask) {r=true}) {name=TireLodSE, nodeColor=<rgba 0.3906 1 0.1001 1>}
  NamedRerouteDeclaration(None=@n2.Normal) {name=NormalIN, nodeColor=<rgba 0.3817 0.439 1 1>}
 ]
}
