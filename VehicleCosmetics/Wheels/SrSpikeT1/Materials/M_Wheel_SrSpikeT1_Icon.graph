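# Translucent unlit material for the SrSpikeT1 vehicle wheel cosmetic's icon, enabled for use on skeletal meshes.
# The Icon texture parameter's RGB drives both Emissive Color and Opacity through material attributes, with no separate alpha used.
# Those attributes pass through MF_BaseVehicle_Damage and then MF_VehicleDeRez_Transition, so the icon runs through the vehicle damage and de-rez functions.
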Material /VehicleCosmetics/Wheels/SrSpikeT1/Materials/M_Wheel_SrSpikeT1_Icon
graph {
 properties {
  bAutomaticallySetUsageInEditor=false
  bUseMaterialAttributes=true
  bUsedWithSkeletalMesh=true
  blendMode=BLEND_Translucent
  shadingModel=MSM_Unlit
  usedShadingModels=Unlit
 }
 nodes [
  n1 = TextureSampleParameter2D() {parameterName=Icon, texture=/VehicleCosmetics/Wheels/SrSpikeT1/Textures/T_S3_01_Icon}
  MaterialFunctionCall(In=MaterialFunctionCall(Material=SetMaterialAttributes!(MaterialAttributes, Emissive Color=@n1.RGB, Opacity=@n1.RGB) {attributeSetTypes [#1, #2]}) {materialFunction=&MF_BaseVehicle_Damage}.Result) {materialFunction=&MF_VehicleDeRez_Transition}
 ]
}
