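# Translucent unlit material for the animated glow effect on the SrSpikeT1 vehicle wheel cosmetic.
# Emissive tints a rotating circle times a static swirl with CustomColor, plus a flat brightness term.
# Opacity, scaled by EmissivePower, sums sparkle from paired dot textures rotating at different speeds and swirl-masked counter-sampled circle channels.
# Material attributes pass through the shared vehicle damage and de-rez transition functions.
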Material /VehicleCosmetics/Wheels/SrSpikeT1/Materials/M_Wheel_SrSpikeT1_SE
graph {
 properties {
  bAutomaticallySetUsageInEditor=false
  bUseMaterialAttributes=true
  bUsedWithSkeletalMesh=true
  blendMode=BLEND_Translucent
  shadingModel=MSM_Unlit
  usedShadingModels=Unlit
 }
 nodes [
  n1 = NamedRerouteUsage()
  n2 = NamedRerouteUsage()
  MaterialFunctionCall(In=MaterialFunctionCall(Material=SetMaterialAttributes!(MaterialAttributes, Emissive Color=@n1, Opacity=@n2) {attributeSetTypes [#1, #2]}) {materialFunction=&MF_BaseVehicle_Damage}.Result) {materialFunction=&MF_VehicleDeRez_Transition}
  n3 = TextureSample(UVs=Rotator() {speed=0.7}) {samplerSource=SSM_Clamp_WorldGroupSettings, texture=/VehicleCosmetics/Wheels/_SharedTextures/FX/CircleMV_RGB_T}
  n4 = TextureSample() {samplerType=SAMPLERTYPE_LinearColor, texture=/VehicleCosmetics/Wheels/Insectoid/Textures/Swirl_RGB_02_T}
  n5 = Multiply(A=@n3.A, B=@n4.A)
  n6 = VectorParameter() {parameterName=CustomColor, defaultValue=<rgba 1 0.1754 0 1>}
  n7 = ScalarParameter() {defaultValue=2, parameterName=EmissiveBrightness}
  NamedRerouteDeclaration(None=Add(A=Multiply(A=@n5, B=@n6.RGB), B=Multiply(A=@n6.RGB, B=@n7))) {name=Emissive, nodeColor=<rgba 0.7812 0 1 1>}
  n8 = TextureSample(UVs=Rotator() {speed=1}) {samplerType=SAMPLERTYPE_LinearColor, texture=/VehicleCosmetics/Wheels/SrSpikeT1/Textures/Dots_RGB_Med_T}
  n9 = TextureSample(UVs=Rotator() {speed=1.1}) {samplerType=SAMPLERTYPE_LinearColor, texture=&Dots_RGB_Med_T}
  n10 = TextureSample(UVs=Rotator() {speed=0.6}) {samplerType=SAMPLERTYPE_LinearColor, texture=&Dots_RGB_Med_T}
  n11 = TextureSample(UVs=Rotator() {speed=0.7}) {samplerType=SAMPLERTYPE_LinearColor, texture=&Dots_RGB_Med_T}
  n12 = Multiply(A=Add(A=Multiply(A=@n8.G, B=@n9.R), B=Multiply(A=@n10.G, B=@n11.R)), B=Constant() {r=5})
  n13 = Rotator() {speed=3.7}
  n14 = Multiply(
   A=Multiply(
    A=@n3.G
    B=TextureSample(UVs=@n13) {samplerSource=SSM_Clamp_WorldGroupSettings, texture=&CircleMV_RGB_T}.R
   )
   B=@n4.A
  )
  NamedRerouteDeclaration(None=Multiply(A=ScalarParameter() {defaultValue=1, parameterName=EmissivePower}, B=Add(A=@n12, B=@n14))) {name=Opacity, nodeColor=<rgba 0.2552 0.2552 0.2552 1>}
 ]
}
